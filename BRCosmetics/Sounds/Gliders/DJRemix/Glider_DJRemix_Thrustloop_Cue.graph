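# Sound cue for the DJ Remix glider's thrust loop, layering a glider loop sound with a music loop.
# A Mixer plays both waves together, with the music at full input volume and the glider layer much quieter.
# Only the music wave player has looping enabled.
# The cue uses the player glider thrust attenuation, the cosmetic glider loop properties sheet, and the gliders music sound class.
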SoundCue /BRCosmetics/Sounds/Gliders/DJRemix/Glider_DJRemix_Thrustloop_Cue
root=Mixer_1
nodes [
 {id=Mixer_1, children [WavePlayer_2,WavePlayer_1], properties {inputVolume [0.2,1]}}
 {id=WavePlayer_1, properties {bLooping=true, soundWaveAssetPtr=&/Athena_Glider_DJRemix_Music_Loop}}
 {id=WavePlayer_2, properties {soundWaveAssetPtr=&/Athena_Glider_DJRemix_Loop}}
]
properties {
 attenuationSettings=&Fort_Player_Glider_Thrust_Athena_Att
 audioPropertiesSheet=&APS_Cosmetic_Glider_Loop
 maxDistance=12500
 soundClassObject=/Game/Sounds/FortSoundClassesAndModes/Fort_Gliders_Music
 volumeMultiplier=0.65
}
